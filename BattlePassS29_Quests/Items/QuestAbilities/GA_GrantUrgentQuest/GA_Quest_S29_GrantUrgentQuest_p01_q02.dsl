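; Gameplay ability that grants a Season 29 urgent quest to the owning player in response to a gameplay event.
; On activation it calls the parent, then waits on the owning pawn for the configured event tag and checks the event's target tags for a required tag.
; It looks up the configured quest on the player state and grants it through the quest instance manager subsystem only if absent, preventing double grants.
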(event Ability|EventActivateAbility
  ;[1 1 1] Check if player already has the urgent quest, then grant if they dont (don't double grant quest)
  (|CallParentFunction)
  (bind (_asynctaskproxy _payload) (Ability|Async|WaitGameplayEventToActor (Variables|GAVars|Data|GetOwningPlayerPawn) (Variables|Default|GetEvent%20Tag))
    (:EventReceived
      (if (GameplayTags|HasTag (.TargetTags (Utilities|Struct|BreakGameplayEventData _payload)) (Variables|Default|GetTag))
        (bind _returnvalue (Quest|GetQuestfromDefinition (Variables|/Script/Engine.Controller|GetPlayerState (Fortnite|GetFortPlayerControllerfromActor (Variables|GAVars|Data|GetOwningPlayerPawn))) (Variables|Default|GetQuest%20Def%20Athena%20to%20Grant)))
        (Utilities|IsValid _returnvalue
          (:Is%20Not%20Valid
            (Utilities|IsValid (Utilities|ToObject(Interface) (Class|QuestInstanceManagerSubsystem|GrantQuest (WorldSubsystems|GetQuestInstanceManagerSubsystem) (Utilities|Struct|MakeGrantQuestBlueprintParameters (Variables|Default|GetQuest%20Def%20Athena%20to%20Grant) (Variables|GAVars|Data|GetOwningPlayerPawn))))
              (:Is%20Valid
                (Development|PrintString "Quest Valid")))))))))

(event Ability|EventOnEndAbility
  (params (bWasCancelled :type Boolean)))

(fn CheckForNearbyEnemies
  (params (LastCheck1 :type Boolean)))
